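# Conversation database for the NPC Mystify, giving its greeting dialogue with heist-themed lines about outsmarting Cryptic.
# The entry tag leads to one speech node with a first-ever-conversation greeting, a default line and flavor variants that share an outfit-reaction priority row but list no requirements.
# That node links out to the shared NPC Library quests and basic services conversations by entry tag.
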ConversationDatabase /NPCLibrary/NPCs/Mystify/NPCLibrary_Conversation_Mystify
entry_points [{tag=Conversation.Entry.Mystify, destinations [#1]}]
nodes [
 ConversationEntryPointNode_7 {
  id=#1
  class=ConversationEntryPointNode
  outputs [#2]
  properties {entryTag {tagName=Conversation.Entry.Mystify}}
 }
 Speech_C_4 {
  id=#2
  class=Speech_C
  outputs [#3, #4]
  properties {
   nodeName=""
   generalConfig {
    defaultMessage=Don't tell Cryptic I'm here. He'll get uppity.
    contextualMessages [
     {
      message=You here for the heist?
      contextRequirements [{requirement {class=FortControllerRequirement_IsFirstEverConversationWithNPC}}]
      priority {curve {rowName=Tandem.Speech.Priority.FirstEverInteraction}}
     }
     {
      message=Don't tell Cryptic I'm here. He'll get uppity.
      requirementMatchPolicy=RequireAny
      contextRequirements []
      priority {curve {curveTable=null, rowName=Tandem.Speech.Priority.OutfitReaction}}
     }
     {
      message=There's nothing I love more than outsmarting Cryptic.
      requirementMatchPolicy=RequireAny
      contextRequirements []
      priority {curve {curveTable=null, rowName=Tandem.Speech.Priority.OutfitReaction}}
     }
     {
      message="If you think about it, a heist is basically performance art."
      requirementMatchPolicy=RequireAny
      contextRequirements []
      priority {curve {curveTable=null, rowName=Tandem.Speech.Priority.OutfitReaction}}
     }
     {
      message=It's not a crime if they can't catch you doing it. Probably.
      requirementMatchPolicy=RequireAny
      contextRequirements []
      priority {curve {curveTable=null, rowName=Tandem.Speech.Priority.OutfitReaction}}
     }
    ]
   }
  }
 }
 ConversationLinkNode_0 {id=#4, class=ConversationLinkNode, properties {remoteEntryTag {tagName=Conversation.Entry.BasicServices}}}
 ConversationLinkNode_1 {id=#3, class=ConversationLinkNode, properties {remoteEntryTag {tagName=Conversation.Entry.NPCLibrary.Quests}}}
]
